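# Cloth-shaded character material for the FinchVisit body, adding stylized, camera-relative rim shading to the 2023 anime shader functions, which include eye animation.
# Three screen-space light vectors, rebuilt from a camera-relative basis, feed hard-edged Phong sphere masks for side, top and bottom rims.
# TechArt masks, stepped Fresnel outlines and top/bottom view detection support the effect, and IsInFrontend supplies a frontend blend.
# Custom shading blends into the anime shading gradient, with a texture-tinted shaded color.
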Material /BRCosmetics/Characters/Player/Male/Medium/Bodies/M_MED_FinchVisit/Materials/M_FinchVisit_RimLight
graph {
 properties {
  bAutomaticallySetUsageInEditor=false
  bUseMaterialAttributes=true
  bUsedWithMeshDeformer=true
  bUsedWithMorphTargets=true
  bUsedWithSkeletalMesh=true
  bUsedWithStaticLighting=true
  pixelDepthOffsetMode=PDOM_Legacy
  shadingModel=MSM_Cloth
  usedShadingModels=Cloth
 }
 nodes [
  Constant()
  n1 = ScalarParameter() {defaultValue=0.2726, group=CustomShading, parameterName=InnerRadius}
  LinearInterpolate(A=@n1, Alpha=NamedRerouteUsage())
  n2 = NamedRerouteUsage()
  n3 = MaterialFunctionCall!(CustomUV) {materialFunction=/Game/Global/Materials/Character_Parent_2023/Anime_Parent_2023/MaterialFunctions/MFC_AnimeShader_2023_Pre}
  n4 = MaterialFunctionCall!(CustomLightVector, OverrideNormal_WS, STT=@n3.STT, Normal_Map=@n3.STT, DFL=@n3.STT, FX=@n3.STT) {materialFunction=/Game/Global/Materials/Character_Parent_2023/Anime_Parent_2023/MaterialFunctions/MFC_AnimeShader_2023_Calculations}
  n5 = VertexColor()
  n6 = TextureSampleParameter2D() {
   group=01_EyeAnimation_Parameters
   parameterName=EyeMask
   samplerSource=SSM_Wrap_WorldGroupSettings
   samplerType=SAMPLERTYPE_Masks
   sortPriority=-32
   texture=&T_WhiteMask
  }
  n7 = StaticSwitchParameter(True=AppendVector(A=@n5, B=@n5), False=@n6.RGBA) {group=01_EyeAnimation_Parameters, parameterName=UseVertexColorForEyeMask, sortPriority=-33}
  n8 = TextureCoordinate() {coordinateIndex=1}
  n9 = Sign(None=Subtract(A=ComponentMask(None=@n8) {r=true}) {constB=0.0001})
  n10 = MaterialFunctionCall!(
   Color Lit=@n3.Color_Lit
   Color Shaded=@n3.Color_Lit
   Use UV2=StaticBoolParameter() {defaultValue=true, group=01_EyeAnimation_Parameters, parameterName=Use UV2 for Eye Animation, sortPriority=-30}
   Mask=StaticSwitchParameter(
    True=ChannelMaskParameter(None=@n7) {group=01_EyeAnimation_Parameters, parameterName=EyeMaskChannel, sortPriority=-32}.R
    False=Saturate(None=@n9)
   ) {group=01_EyeAnimation_Parameters, parameterName=UseEyeMask, sortPriority=-33}
  ) {materialFunction=/Game/Global/Materials/Character_Parent_2023/Anime_Parent_2023/MaterialFunctions/MF_AnimeEyeAnimation_v2}
  MaterialFunctionCall!(
   Shading_Gradient=LinearInterpolate(A=Saturate(None=@n2), B=@n4.Shading_Gradient, Alpha=NamedRerouteUsage())
   Threshold_Gradient=@n4.Threshold_Gradient
   Specular_Gradient=@n4.Threshold_Gradient
   Sheen_Mask=@n4.Threshold_Gradient
   Thickness=@n4.Threshold_Gradient
   DFL_Mask=@n4.Threshold_Gradient
   FX=@n4.Threshold_Gradient
   Metallic_Gradient=@n4.Threshold_Gradient
   RimLight_Gradient=@n4.Threshold_Gradient
   NormalBuffer=@n4.Threshold_Gradient
   --------
   CINE_LightSources=@n4.Threshold_Gradient
   CINE_LightColor=@n4.Threshold_Gradient
   ---------
   Color_Lit=@n10.Lit
   Color_Shaded=LinearInterpolate(
    A=TextureSample() {texture=/BRCosmetics/Characters/Player/Male/Medium/Bodies/M_MED_FinchVisit/Textures/T_FinchVisit_Body_CS_TechArt}.RGB
    B=@n10.Shaded
    Alpha=NamedRerouteUsage()
   )
   SRM=@n3.SRM
   Emissive=@n3.SRM
   DFL_Color=@n3.SRM
   Specular_Color=@n3.SRM
   RimLight_Color=@n3.SRM
  ) {materialFunction=/Game/Global/Materials/Character_Parent_2023/Anime_Parent_2023/MaterialFunctions/MFC_AnimeShader_2023_Post}
  NamedRerouteDeclaration(None=Transform(None=@n3.Normal)) {name=WorldNormal, nodeColor=<rgba 1 0 0.3281 1>}
  n11 = Normalize(VectorInput=VectorParameter() {group=CustomShading, parameterName=ScreenSpaceLightVector02, defaultValue=<rgba 0 1 -0.05 1>}.RGB)
  n12 = Constant3Vector() {constant=<rgba 0 1 0 0>}
  n13 = Normalize(VectorInput=WorldPosition() {worldPositionShaderOffset=WPT_CameraRelative}.XYZ)
  n14 = Normalize(VectorInput=CrossProduct(A=Transform(None=@n12) {transformSourceType=TRANSFORMSOURCE_View}, B=@n13))
  n15 = Multiply(A=ComponentMask(None=@n11) {r=true}, B=@n14)
  n16 = Multiply(A=ComponentMask(None=@n11) {g=true}, B=CrossProduct(A=@n13, B=@n14))
  n17 = ComponentMask(None=@n11) {b=true}
  NamedRerouteDeclaration(None=Add(A=Add(A=@n15, B=@n16), B=Multiply(A=@n17, B=@n13))) {name=LightVector02, nodeColor=<rgba 0 1 0.7891 1>}
  n18 = TextureSampleParameter2D() {
   group=CustomShading
   parameterName=TechArtMask_Add
   samplerType=SAMPLERTYPE_Masks
   texture=/BRCosmetics/Characters/Player/Male/Medium/Bodies/M_MED_FinchVisit/Textures/T_FinchVisitMask_Body_TechArt_Add
  }
  NamedRerouteDeclaration(None=@n18.R) {name=TopMask, nodeColor=<rgba 0.1719 0 1 1>}
  NamedRerouteDeclaration(None=ScalarParameter() {parameterName=IsInFrontend}) {name=FE_Alpha, nodeColor=<rgba 0.2656 1 0 1>}
  n19 = Transform(None=Constant3Vector() {constant=<rgba 0 0 1 0>}) {transformSourceType=TRANSFORMSOURCE_View}
  n20 = Saturate(None=DotProduct(A=@n19, B=Constant3Vector() {constant=<rgba 0 0 -1 0>}))
  NamedRerouteDeclaration(None=SmoothStep(Value=Power(Base=@n20) {constExponent=5}) {constMax=0.5, constMin=0.1}) {name=TopView, nodeColor=<rgba 1 0.9375 0 1>}
  n21 = NamedRerouteUsage()
  n22 = Multiply(A=NamedRerouteUsage(), B=VectorParameter() {group=CustomShading, parameterName=NormalMulti01, defaultValue=<rgba 1 1 1 1>}.RGB)
  n23 = Saturate(None=DotProduct(A=NamedRerouteUsage(), B=Normalize(VectorInput=@n22)) {desc=Phong})
  n24 = ScalarParameter() {defaultValue=0.5, group=CustomShading, parameterName=InnerPOW}
  n25 = ScalarParameter() {defaultValue=0.2268, group=CustomShading, parameterName=InnerOffset02}
  n26 = Multiply(
   A=SphereMask(
    A=Power(Base=@n23, Exp=@n24)
    B=Constant() {r=1}
    Radius=@n1
    Hardness=ScalarParameter() {defaultValue=0.95, group=CustomShading, parameterName=InnerHardness}
   ) {attenuationRadius=0.1, hardnessPercent=0.3}
   B=Step(Y=Add(A=@n23, B=@n25))
  )
  n27 = NamedRerouteUsage()
  NamedRerouteDeclaration(None=Multiply(A=Multiply(A=@n21, B=@n26), B=OneMinus(None=@n27))) {name=SideRim, nodeColor=<rgba 1 0.0469 0 1>}
  n28 = NamedRerouteUsage()
  n29 = NamedRerouteUsage()
  n30 = VectorParameter() {group=CustomShading, parameterName=NormalMulti02, defaultValue=<rgba 1 1 1 1>}
  n31 = DotProduct(A=NamedRerouteUsage(), B=Normalize(VectorInput=Multiply(A=@n29, B=@n30.RGB))) {desc=Phong}
  n32 = SphereMask(
   A=Power(Base=Saturate(None=@n31), Exp=ScalarParameter() {defaultValue=1, group=CustomShading, parameterName=InnerPOW01})
   B=Constant() {r=1}
   Radius=LinearInterpolate(
    A=ScalarParameter() {defaultValue=0.3391, group=CustomShading, parameterName=InnerRadius01}
    B=ScalarParameter() {defaultValue=1.1, group=CustomShading, parameterName=InnerRadius01_less}
    Alpha=NamedRerouteUsage()
   )
   Hardness=ScalarParameter() {defaultValue=0.95, group=CustomShading, parameterName=InnerHardness01}
  ) {attenuationRadius=0.1, hardnessPercent=0.3}
  NamedRerouteDeclaration(None=Multiply(A=Multiply(A=@n28, B=@n32), B=NamedRerouteUsage())) {name=TopRim, nodeColor=<rgba 0.1953 0 1 1>}
  n33 = NamedRerouteUsage()
  n34 = NamedRerouteUsage()
  n35 = VectorParameter() {group=CustomShading, parameterName=NormalMulti03, defaultValue=<rgba 1 1 1 1>}
  n36 = DotProduct(A=NamedRerouteUsage(), B=Normalize(VectorInput=Multiply(A=@n34, B=@n35.RGB))) {desc=Phong}
  n37 = SphereMask(
   A=Power(Base=Saturate(None=@n36), Exp=ScalarParameter() {defaultValue=1, group=CustomShading, parameterName=InnerPOW02})
   B=Constant() {r=1}
   Radius=LinearInterpolate(B=ScalarParameter() {defaultValue=0.3, group=CustomShading, parameterName=InnerRadius02}, Alpha=NamedRerouteUsage())
   Hardness=ScalarParameter() {defaultValue=0.95, group=CustomShading, parameterName=InnerHardness02}
  ) {attenuationRadius=0.1, hardnessPercent=0.3}
  NamedRerouteDeclaration(None=Multiply(A=Multiply(A=@n33, B=@n37), B=NamedRerouteUsage())) {name=BottomRim, nodeColor=<rgba 0.1953 0 1 1>}
  n38 = Saturate(None=DotProduct(A=@n19, B=Constant3Vector() {constant=<rgba 0 0 1 0>}))
  NamedRerouteDeclaration(None=SmoothStep(Value=Power(Base=@n38) {constExponent=5}) {constMax=0.5}) {name=BottomView, nodeColor=<rgba 0 1 0.5313 1>}
  n39 = Fresnel(Normal=NamedRerouteUsage()) {baseReflectFraction=0, exponent=1}
  n40 = Power(Base=@n39, Exp=ScalarParameter() {defaultValue=1, group=CustomShading, parameterName=OUTlinePOW01})
  n41 = ScalarParameter() {defaultValue=0.5544, group=CustomShading, parameterName=OUTLineOffset01}
  NamedRerouteDeclaration(None=Step(Y=Add(A=@n40, B=@n41))) {name=OffsetFresnelOUTline, nodeColor=<rgba 0.4062 0 1 1>}
  n42 = Normalize(VectorInput=LinearInterpolate(
    A=VectorParameter() {group=CustomShading, parameterName=ScreenSpaceLightVector01, defaultValue=<rgba -1 -0.5 -0.1 1>}.RGB
    B=VectorParameter() {group=CustomShading, parameterName=ScreenSpaceLightVector01_FE, defaultValue=<rgba -1 0.2 -0.1 1>}.RGB
    Alpha=NamedRerouteUsage()
   ))
  n43 = Constant3Vector() {constant=<rgba 0 1 0 0>}
  n44 = Normalize(VectorInput=WorldPosition() {worldPositionShaderOffset=WPT_CameraRelative}.XYZ)
  n45 = Normalize(VectorInput=CrossProduct(A=Transform(None=@n43) {transformSourceType=TRANSFORMSOURCE_View}, B=@n44))
  n46 = Multiply(A=ComponentMask(None=@n42) {r=true}, B=@n45)
  n47 = Multiply(A=ComponentMask(None=@n42) {g=true}, B=CrossProduct(A=@n44, B=@n45))
  n48 = ComponentMask(None=@n42) {b=true}
  NamedRerouteDeclaration(None=Add(A=Add(A=@n46, B=@n47), B=Multiply(A=@n48, B=@n44))) {name=LightVector01, nodeColor=<rgba 0 1 0.7891 1>}
  n49 = NamedRerouteUsage()
  n50 = NamedRerouteUsage()
  n51 = NamedRerouteUsage()
  n52 = NamedRerouteUsage()
  NamedRerouteDeclaration(None=Max(A=Max(A=@n49, B=@n50), B=Multiply(A=@n51, B=@n52))) {name=CustomShading, nodeColor=<rgba 0.2422 0 1 1>}
  n53 = Power(Base=@n39, Exp=ScalarParameter() {defaultValue=0.5, group=CustomShading, parameterName=OUTlinePOW02})
  n54 = LinearInterpolate(A=ScalarParameter() {defaultValue=0.2837, group=CustomShading, parameterName=OUTlineOffset02}, Alpha=NamedRerouteUsage()) {constB=0.2}
  NamedRerouteDeclaration(None=Step(Y=Add(A=@n53, B=@n54))) {name=OffsetFresnelOUTline01, nodeColor=<rgba 0.4062 0 1 1>}
  n55 = Normalize(VectorInput=VectorParameter() {group=CustomShading, parameterName=ScreenSpaceLightVector03, defaultValue=<rgba 0 -1 -0.1 1>}.RGB)
  n56 = Constant3Vector() {constant=<rgba 0 1 0 0>}
  n57 = Normalize(VectorInput=WorldPosition() {worldPositionShaderOffset=WPT_CameraRelative}.XYZ)
  n58 = Normalize(VectorInput=CrossProduct(A=Transform(None=@n56) {transformSourceType=TRANSFORMSOURCE_View}, B=@n57))
  n59 = Multiply(A=ComponentMask(None=@n55) {r=true}, B=@n58)
  n60 = Multiply(A=ComponentMask(None=@n55) {g=true}, B=CrossProduct(A=@n57, B=@n58))
  n61 = ComponentMask(None=@n55) {b=true}
  NamedRerouteDeclaration(None=Add(A=Add(A=@n59, B=@n60), B=Multiply(A=@n61, B=@n57))) {name=LightVector03, nodeColor=<rgba 0 1 0.7891 1>}
  n62 = TextureSampleParameter2D() {
   group=CustomShading
   parameterName=TechArtMask
   samplerType=SAMPLERTYPE_Masks
   texture=/BRCosmetics/Characters/Player/Male/Medium/Bodies/M_MED_FinchVisit/Textures/T_FinchVisitMask_Body_TechArt
  }
  NamedRerouteDeclaration(None=@n62.R) {name=AreaMask, nodeColor=<rgba 0 0.1094 1 1>}
  NamedRerouteDeclaration(None=@n62.G) {name=FXMask, nodeColor=<rgba 0.4297 1 0 1>}
  NamedRerouteDeclaration(None=@n62.B) {name=LimitMask, nodeColor=<rgba 1 0.9375 0 1>}
  NamedRerouteUsage()
 ]
}
